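# Unlit alpha-composite material, apparently a recovery copy, for painting obstacles into a 3D fluid simulation volume laid out as tiled XY frames, allowing negative emissive.
# Decoded volume coordinates drive a procedural fan with radial, twisted blades and a flared inlet ring, combined with a volume edge mask.
# Other branches scale blade masks by a rotated position for velocity and draw a line along a velocity vector.
# No branch connects to a material output in this file.
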Material /Volumetrics/Tools/FluidSimulation/Materials/3D/recover/M_Obstacle_3D_2019-04-29-13-35-37.M_Obstacle_3D
graph {
 properties {
  bAllowNegativeEmissiveColor=true
  blendMode=BLEND_AlphaComposite
  pixelDepthOffsetMode=PDOM_Legacy
  shadingModel=MSM_Unlit
  usedShadingModels=Unlit
 }
 nodes [
  n1 = MaterialFunctionCall!(UVs, XY Frames=ComponentMask(None=VectorParameter() {parameterName=XYFrames, defaultValue=<rgba 16 16 0 0>}.RGB) {g=true, r=true}) {materialFunction=/Volumetrics/Content/VolumeTextures/MaterialFunctions/EncodeVolumeCoordinates}
  n2 = VectorParameter() {parameterName=P, defaultValue=<rgba 0.5 0.5 0.2 0>}
  n3 = Divide(A=VectorParameter() {parameterName=V}.RGB) {constB=-256}
  n4 = ScalarParameter() {defaultValue=0.05, parameterName=R}
  Multiply(
   A=VectorParameter() {parameterName=V}.RGB
   B=MaterialFunctionCall!(Coordinates=@n1.Result, P0=@n2.RGB, P1=Add(A=@n2.RGB, B=@n3), Line Width=@n4, Edge Falloff, Square Corners?=StaticBool()) {materialFunction=/Engine/Functions/Engine_MaterialFunctions01/Coordinates/DrawLine-3D}.Result
  )
  n5 = MaterialFunctionCall(XY Frames=ComponentMask(None=VectorParameter() {parameterName=XYFrames_1, defaultValue=<rgba 16 16 0 0>}.RGB) {g=true, r=true}) {materialFunction=&EncodeVolumeCoordinates}
  n6 = MaterialFunctionCall!(
   UVW=AppendVector(A=ComponentMask(None=@n5.Result) {g=true, r=true}, B=Constant() {r=0.5})
   Falloff Width=ScalarParameter() {defaultValue=0.02, parameterName=FW}
   Hardness=ScalarParameter() {defaultValue=1, parameterName=hardness}
   Falloff Power=ScalarParameter() {defaultValue=2, parameterName=FP}
   XY Frames=ScalarParameter() {defaultValue=10, parameterName=XYFrames}
  ) {materialFunction=/Volumetrics/Content/VolumeTextures/MaterialFunctions/VolumeEdgeMask}
  Multiply(A=@n6.Result)
  n7 = ScalarParameter() {defaultValue=-0.1, parameterName=rasdfasdfas}
  n8 = MaterialFunctionCall(XY Frames=ComponentMask(None=VectorParameter() {parameterName=XYFrames, defaultValue=<rgba 16 16 0 0>}.RGB) {g=true, r=true}) {materialFunction=&EncodeVolumeCoordinates}
  n9 = SphereMask(A=ComponentMask(None=@n8.Result) {g=true, r=true}, B=Constant() {r=0.5}, Radius=Constant() {r=0.5}) {hardnessPercent=0}
  n10 = Subtract(A=@n8.Result) {constB=0.5}
  n11 = RotateAboutAxis(
   NormalizedRotationAxis=Constant3Vector() {constant=<rgba 0 0 1 0>}
   RotationAngle=Add(A=ScalarParameter() {parameterName=Rot}, B=Multiply(A=@n7, B=@n9))
   PivotPoint=Constant()
   Position=@n10
  )
  n12 = ComponentMask(None=Add(A=Add(A=@n11, B=@n10)) {constB=0.5}) {g=true, r=true}
  n13 = MaterialFunctionCall(Vector or UVs=Normalize(VectorInput=Subtract(A=@n12) {constB=0.5})) {materialFunction=&VectorToRadialValue}
  n14 = ScalarParameter() {defaultValue=6, parameterName=Blades}
  n15 = MaterialFunctionCall!(Alpha=Frac(None=Multiply(A=@n13.Vector Converted to Angle, B=@n14)), Offset=ScalarParameter() {defaultValue=0.5, parameterName=BladeGap}) {materialFunction=/Engine/Functions/Engine_MaterialFunctions01/Opacity/AlphaOffset}
  n16 = Saturate(None=@n15.Result)
  n17 = ScalarParameter() {defaultValue=1.5, parameterName=ScoopPower}
  n18 = Multiply(A=Subtract(A=Power(Base=@n16, Exp=@n17)) {constB=0.5}, B=ScalarParameter() {defaultValue=0.2, parameterName=Blade Height})
  n19 = ComponentMask(None=@n8.Result) {b=true}
  n20 = Ceil(None=@n15.Result)
  n21 = MaterialFunctionCall(XY Frames=ComponentMask(None=VectorParameter() {parameterName=XYFrames_2, defaultValue=<rgba 16 16 0 0>}.RGB) {g=true, r=true}) {materialFunction=&EncodeVolumeCoordinates}
  n22 = ComponentMask(None=@n21.Result) {g=true, r=true}
  n23 = MaterialFunctionCall(Vector 2=Multiply(A=Subtract(B=@n22) {constA=0.5, constB=0.5}) {constB=2}) {materialFunction=/Engine/Functions/Engine_MaterialFunctions02/Utility/VectorLength}
  n24 = ScalarParameter() {defaultValue=0.85, parameterName=FanRadius}
  n25 = Floor(None=Add(A=@n23.V2 Length, B=OneMinus(None=@n24)))
  n26 = Multiply(
   A=SphereMask(A=Add(A=@n18) {constB=0.5}, B=Multiply(A=@n19, B=@n20), Radius=ScalarParameter() {defaultValue=0.02, parameterName=BladeWidth})
   B=OneMinus(None=Saturate(None=@n25))
  )
  n27 = Constant3Vector() {constant=<rgba 0 0 1 0>}
  n28 = ScalarParameter() {defaultValue=0.05, parameterName=RotRate}
  n29 = Constant()
  Multiply(A=@n26, B=Multiply(A=RotateAboutAxis(NormalizedRotationAxis=@n27, RotationAngle=@n28, PivotPoint=@n29, Position=@n10)) {constB=-256})
  Multiply(A=@n5.Result, B=Constant3Vector() {constant=<rgba 1 1 0.9 0>})
  n30 = ScalarParameter() {defaultValue=0.95, parameterName=InletRadius}
  n31 = ScalarParameter() {defaultValue=1, parameterName=InletRadius_1}
  n32 = ComponentMask(None=@n21.Result) {b=true}
  n33 = Subtract(A=@n32) {constB=0.5}
  n34 = ScalarParameter() {defaultValue=0.2, parameterName=Inlet Flat Height}
  n35 = MaterialFunctionCall(Alpha=Multiply(A=Abs(None=@n33)) {constB=2}, Offset=@n34) {materialFunction=&AlphaOffset}
  n36 = Add(A=@n23.V2 Length, B=OneMinus(None=LinearInterpolate(A=@n30, B=@n31, Alpha=@n35.Result)))
  n37 = Multiply(
   A=Saturate(None=Floor(None=@n36))
   B=SphereMask(A=@n32, B=Constant() {r=0.5}, Radius=ScalarParameter() {defaultValue=0.5, parameterName=Inlet Height})
  )
  OneMinus(None=Max(A=Max(A=@n26, B=@n37), B=OneMinus(None=@n6.Result)))
  OneMinus(None=@n34)
  SphereMask(A=@n1.Result, B=@n2.RGB, Radius=@n4)
  TextureCoordinate()
 ]
}
